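# Material function for the male turtleneck body that shifts UVs horizontally by view angle, producing a parallax-like layer offset.
# The local-space camera X component, blended by masks, is added to U and UVOffsetX, scaled by Depth, and added to TexCoord.
# Hard and soft sphere masks around a fixed U position apply a sign-flipped shift scaled by Depth_SideMult for smooth edges.
# The hard mask and a V threshold switch to the unscaled camera shift.
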MaterialFunction /Game/Characters/Player/Male/Medium/Bodies/M_MED_Turtleneck/Materials/MF_TN_LayerOffset
graph {nodes [
  n1 = TextureCoordinate()
  n2 = MaterialFunctionCall(Float2=@n1) {materialFunction=&BreakOutFloat2Components}
  n3 = FunctionInput() {inputName=UVOffsetX, inputType=FunctionInput_Scalar}
  n4 = CameraVectorWS()
  n5 = MaterialFunctionCall(Color=Multiply(
    A=Transform(None=@n4) {transformSourceType=TRANSFORMSOURCE_World, transformType=TRANSFORM_Local}
    B=Constant3Vector() {constant=<rgba -1 1 1 0>}
   ) {constB=-1}) {materialFunction=&SplitComponents}
  n6 = FunctionInput(Preview=Constant() {r=1.2}) {
   desc=Thicken the line on the side to create a smooth transition around mask endge
   description=Thicken the line on the side to create a smooth transition around mask endge
   inputName=Depth_SideMult
   inputType=FunctionInput_Scalar
   sortPriority=2
  }
  n7 = Constant() {r=0.764}
  n8 = SphereMask(A=@n2.R, B=@n7, Radius=ScalarParameter() {defaultValue=0.082, parameterName=OffsetRadius}, Hardness=Constant() {r=1})
  n9 = Power(Base=SphereMask(A=@n2.R, B=@n7, Radius=Constant() {r=0.4}, Hardness=Constant() {r=0.2})) {constExponent=5}
  n10 = LinearInterpolate(B=Multiply(A=Multiply(A=@n5.R) {constB=-1}, B=@n6), Alpha=Saturate(None=Add(A=@n8, B=@n9)))
  n11 = Constant() {r=0.84}
  n12 = Constant() {r=1}
  n13 = Constant()
  n14 = Saturate(None=Add(A=If(A=@n2.G, B=@n11, "A > B"=@n12, "A == B"=@n12, "A < B"=@n13), B=@n8))
  n15 = AppendVector(A=Add(A=Add(A=@n2.R, B=@n3), B=LinearInterpolate(A=@n10, B=@n5.R, Alpha=@n14)), B=Constant())
  n16 = FunctionInput(Preview=Constant() {r=0.02}) {inputName=Depth, inputType=FunctionInput_Scalar, sortPriority=2}
  FunctionOutput(None=Add(A=FunctionInput(Preview=@n1) {inputName=TexCoord, inputType=FunctionInput_Vector2, sortPriority=1}, B=Multiply(A=@n15, B=@n16))) {outputName=TexCoord}
  Multiply(A=@n5.B) {constB=0.2}
 ]}
