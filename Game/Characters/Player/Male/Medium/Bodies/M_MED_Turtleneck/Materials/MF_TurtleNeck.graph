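# Material function for the male turtleneck outfit that drives emissive color from an animated distance-field flipbook.
# Time or an input value, optionally event-audio amplitude, picks two frames, whose blend is thresholded into crisp shapes.
# Two depth-offset layers add fake depth, with Fresnel edge falloff and emissive tinting.
# Mobile keeps only the base layer with clamped brightness, and a static switch outputs the raw texture instead.
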MaterialFunction /Game/Characters/Player/Male/Medium/Bodies/M_MED_Turtleneck/Materials/MF_TurtleNeck
graph {
 properties {description=10}
 nodes [
  n1 = TextureCoordinate()
  n2 = TextureSampleParameter2D(UVs=@n1) {
   group=TurtleNeck
   parameterName=DF Texture
   samplerSource=SSM_Wrap_WorldGroupSettings
   samplerType=SAMPLERTYPE_LinearGrayscale
   sortPriority=2
   texture=/Game/Characters/Player/Male/Medium/Bodies/M_MED_Turtleneck/Textures/T_TurtleNeck_Flipbook_DF
  }
  n3 = Multiply(A=Time(), B=ScalarParameter() {defaultValue=2, group=TurtleNeck, parameterName=TN_Frequency, sortPriority=3})
  n4 = Frac(None=@n3)
  n5 = VectorParameter() {group=TurtleNeck, parameterName="TN_SubImages(V2)", sortPriority=2, defaultValue=<rgba 2 2 0 1>}
  n6 = MaterialFunctionCall(Float2=@n5.RGB) {materialFunction=&BreakOutFloat2Components}
  n7 = Multiply(A=@n6.R, B=@n6.R)
  n8 = Multiply(
   A=StaticSwitchParameter(
    True=CollectionParameter() {collection=/Game/Athena/Events/AudioAnalysis/MPC_Audio_Reactive_Parameters, parameterName=Amplitude_VO_B}
    False=ScalarParameter() {group=TurtleNeck, parameterName=TN_AnimatedInput, sliderMax=1}
   ) {group=TurtleNeck, parameterName=UseEventAudio}
   B=@n7
  )
  n9 = Frac(None=@n8)
  n10 = Subtract(A=@n8, B=@n9)
  n11 = StaticSwitchParameter(True=@n10, False=Fmod(A=Subtract(A=@n3, B=@n4), B=@n7)) {group=TurtleNeck, parameterName=TN_UseAnimatedInput, sortPriority=1}
  n12 = Fmod(A=@n11, B=@n6.R)
  n13 = Floor(None=Divide(A=@n11, B=@n6.R))
  n14 = TextureCoordinate()
  n15 = ComponentMask(None=@n5.RGB) {g=true, r=true}
  n16 = Divide(A=Add(A=AppendVector(A=@n12, B=@n13), B=@n14), B=@n15) {constB=3}
  n17 = TextureSampleParameter2D(UVs=@n16) {
   group=TurtleNeck
   parameterName=DF Texture
   samplerType=SAMPLERTYPE_LinearGrayscale
   sortPriority=2
   texture=&T_TurtleNeck_Flipbook_DF
  }
  n18 = Saturate(None=Add(A=@n17.RGB) {constB=-0.3})
  n19 = Add(A=Power(Base=OneMinus(None=@n18)) {constExponent=1.5}) {constB=0.45}
  n20 = GetMaterialAttributes(None=FunctionInput() {inputType=FunctionInput_MaterialAttributes}) {attributeGetTypes [#1]}
  n21 = Fresnel(ExponentIn=Constant() {r=1.5}, BaseReflectFractionIn=Constant(), Normal=Transform(None=@n20.Normal))
  n22 = Subtract(A=@n3, B=Frac(None=@n3))
  n23 = MaterialFunctionCall(Float2=@n5.RGB) {materialFunction=&BreakOutFloat2Components}
  n24 = StaticSwitchParameter(True=Add(A=@n10), False=Fmod(A=Add(A=@n22), B=Multiply(A=@n23.R, B=@n23.R))) {group=TurtleNeck, parameterName=TN_UseAnimatedInput, sortPriority=1}
  n25 = Fmod(A=@n24, B=@n23.R)
  n26 = Floor(None=Divide(A=@n24, B=@n23.R))
  n27 = Divide(A=Add(A=@n14, B=AppendVector(A=@n25, B=@n26)), B=@n15) {constB=3}
  n28 = TextureSampleParameter2D(UVs=@n27) {
   group=TurtleNeck
   parameterName=DF Texture
   samplerType=SAMPLERTYPE_LinearGrayscale
   sortPriority=2
   texture=&T_TurtleNeck_Flipbook_DF
  }
  n29 = StaticSwitchParameter(True=@n9, False=@n4) {group=TurtleNeck, parameterName=TN_UseAnimatedInput, sortPriority=1}
  n30 = LinearInterpolate(A=@n17.RGB, B=@n28.RGB, Alpha=@n29)
  n31 = ScalarParameter() {defaultValue=-0.48, group=TurtleNeck, parameterName=TN_SurfaceDistanceOffset, sortPriority=11}
  n32 = Multiply(A=Saturate(None=Add(A=@n30, B=@n31) {constB=-0.25}), B=Constant() {r=14})
  n33 = ScalarParameter() {defaultValue=0.5, group=TurtleNeck, parameterName=TN_Ly1Strengh}
  n34 = ScalarParameter() {group=TurtleNeck, parameterName=TN_UVOffsetX, sortPriority=22}
  n35 = TextureSampleParameter2D(UVs=MaterialFunctionCall!(
    UVOffsetX=@n34
    TexCoord=@n16
    Depth_SideMult=ScalarParameter() {defaultValue=1.2, group=TurtleNeck, parameterName=TN_Ly1Depth_SideMult}
    Depth=ScalarParameter() {defaultValue=0.02, group=TurtleNeck, parameterName=TN_Ly1Depth}
   ) {materialFunction=&/MF_TN_LayerOffset}.TexCoord) {
   group=TurtleNeck
   parameterName=DF Texture
   samplerSource=SSM_Wrap_WorldGroupSettings
   samplerType=SAMPLERTYPE_LinearGrayscale
   sortPriority=2
   texture=&T_TurtleNeck_Flipbook_DF
  }
  n36 = TextureSampleParameter2D(UVs=MaterialFunctionCall(
    UVOffsetX=@n34
    TexCoord=@n27
    Depth_SideMult=ScalarParameter() {defaultValue=1.2, group=TurtleNeck, parameterName=TN_Ly1Depth_SideMult}
    Depth=ScalarParameter() {defaultValue=0.02, group=TurtleNeck, parameterName=TN_Ly1Depth}
   ) {materialFunction=&MF_TN_LayerOffset}.TexCoord) {
   group=TurtleNeck
   parameterName=DF Texture
   samplerSource=SSM_Wrap_WorldGroupSettings
   samplerType=SAMPLERTYPE_LinearGrayscale
   sortPriority=2
   texture=&T_TurtleNeck_Flipbook_DF
  }
  n37 = Add(
   A=LinearInterpolate(A=@n35.RGB, B=@n36.RGB, Alpha=@n29)
   B=ScalarParameter() {defaultValue=-0.4, group=TurtleNeck, parameterName=TN_Ly1DistanceOffset}
  ) {constB=-0.25}
  n38 = Saturate(None=Multiply(A=Saturate(None=@n37), B=Constant() {r=14}))
  n39 = TextureSampleParameter2D(UVs=MaterialFunctionCall(
    UVOffsetX=@n34
    TexCoord=@n16
    Depth_SideMult=ScalarParameter() {defaultValue=1.2, group=TurtleNeck, parameterName=TN_Ly2Depth_SideMult}
    Depth=ScalarParameter() {defaultValue=0.02, group=TurtleNeck, parameterName=TN_Ly2Depth}
   ) {materialFunction=&MF_TN_LayerOffset}.TexCoord) {
   group=TurtleNeck
   parameterName=DF Texture
   samplerSource=SSM_Wrap_WorldGroupSettings
   samplerType=SAMPLERTYPE_LinearGrayscale
   sortPriority=2
   texture=&T_TurtleNeck_Flipbook_DF
  }
  n40 = TextureSampleParameter2D(UVs=MaterialFunctionCall(
    UVOffsetX=@n34
    TexCoord=@n27
    Depth_SideMult=ScalarParameter() {defaultValue=1.2, group=TurtleNeck, parameterName=TN_Ly2Depth_SideMult}
    Depth=ScalarParameter() {defaultValue=1.2, group=TurtleNeck, parameterName=TN_Ly2Depth}
   ) {materialFunction=&MF_TN_LayerOffset}.TexCoord) {
   group=TurtleNeck
   parameterName=DF Texture
   samplerSource=SSM_Wrap_WorldGroupSettings
   samplerType=SAMPLERTYPE_LinearGrayscale
   sortPriority=2
   texture=&T_TurtleNeck_Flipbook_DF
  }
  n41 = Add(
   A=LinearInterpolate(A=@n39.RGB, B=@n40.RGB, Alpha=@n29)
   B=ScalarParameter() {defaultValue=-0.6, group=TurtleNeck, parameterName=TN_Ly2DistanceOffset}
  ) {constB=-0.25}
  n42 = Saturate(None=Multiply(A=Saturate(None=@n41), B=Constant() {r=14}))
  n43 = ScalarParameter() {defaultValue=0.3, group=TurtleNeck, parameterName=TN_Ly2Strengh}
  n44 = Add(A=Saturate(None=@n32), B=Add(A=Multiply(A=@n33, B=@n38), B=Multiply(A=@n42, B=@n43)))
  n45 = Multiply(
   A=Power(Base=OneMinus(None=@n21)) {constExponent=3}
   B=ShadingPathSwitch(Default=@n44, Deferred=@n44, Mobile=@n32) {desc=Migrated from FeatureLevelSwitch}
  )
  n46 = ScalarParameter() {defaultValue=1, group=TurtleNeck, parameterName=TN_Emissive mult, sortPriority=5}
  n47 = Clamp(None=@n46, Max=ScalarParameter() {defaultValue=6, group=TurtleNeck, parameterName=TN_emissive mult_MobileMax, sortPriority=7})
  n48 = Multiply(
   A=Multiply(A=Multiply(A=@n19, B=@n45), B=ShadingPathSwitch(Default=@n46, Deferred=@n46, Mobile=@n47) {desc=Migrated from FeatureLevelSwitch})
   B=VectorParameter() {group=TurtleNeck, parameterName=TN_Emissive Color, sortPriority=8, defaultValue=<rgba 1 1 1 1>}.RGB
  )
  FunctionOutput(None=SetMaterialAttributes!(MaterialAttributes=@n20.MaterialAttributes, EmissiveColor=StaticSwitchParameter(True=@n2.RGB, False=@n48) {parameterName=Static Texture}) {attributeSetTypes [#2]})
  Multiply(A=Add(A=@n1, B=Constant2Vector() {g=-1, r=-1})) {constB=0.5}
  Reroute(None=@n39.RGB)
  ScalarParameter() {defaultValue=-0.1, group=TurtleNeck, parameterName=TN_UVOffsetY, sortPriority=22}
 ]
}
