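# Master material for the medium male Turtleneck player character body with mask FX, using material attributes on skeletal meshes with morph target and mesh deformer support.
# It is built as a chain of material functions: the shared character PreFX base feeds the character-specific MF_TurtleNeck layer, then CharacterSkinFX.
# The shared PostFX base follows, taking the skin FX result plus PreFX outputs, and feeds a final cinematic character post-FX function.
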Material /Game/Characters/Player/Male/Medium/Bodies/M_MED_Turtleneck/Materials/M_MED_Turtleneck_MaskFX_Master
graph {
 properties {
  bUseMaterialAttributes=true
  bUsedWithMeshDeformer=true
  bUsedWithMorphTargets=true
  bUsedWithSkeletalMesh=true
  pixelDepthOffsetMode=PDOM_Legacy
  refractionMethod=RM_IndexOfRefraction
  usedShadingModels=Default Lit
 }
 nodes [
  n1 = MaterialFunctionCall() {materialFunction=&BaseFN_CharacterFunction_PreFX}
  n2 = MaterialFunctionCall(
   AO=@n1.AO
   Distance Blend=@n1.AO
   Fresnel=@n1.AO
   Gmap=@n1.AO
   Pre Skinned Local Position=@n1.AO
   Material=MaterialFunctionCall(
    Gmap=@n1.Gmap
    MaterialAttributes=MaterialFunctionCall!(In=@n1.Result) {materialFunction=&/MF_TurtleNeck}.Result
    AO=@n1.Gmap
   ) {materialFunction=&CharacterSkinFX}.Result
  ) {materialFunction=&BaseFN_CharacterFunction_PostFX}
  MaterialFunctionCall(In=@n2.Result) {materialFunction=&CINE_MF_CharPostFX}
 ]
}
